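# Niagara system for the Drift Horror backpack cosmetic, adding lightning, ribbon tails and smoke puffs.
# Two distance-culled lightning emitters flash short-lived meshes and lights at random skeletal-mesh bone positions; one stretches arcs between two bone-pair points.
# Two ribbon tails spawn from mesh sampling, scaling rate and lifetime with owner speed; `User.isSkydiving` zeroes spawning and alters their motion.
# Sub-UV smoke sprites spawn around a cylinder, grow, sink and fade.
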NiagaraSystem /BRCosmetics/Accessories/FORT_Backpacks/Backpack_M_MED_Drift_Horror/FX/NS_DriftHorror_BackPack
dependencies {
 usedDataInterfaces [&NiagaraDataInterfaceArrayInt32, &NiagaraDataInterfaceCurve]
 usedDynamicInputs []
 usedModules [&SystemState]
}
emitters {
 Lightning_Spike_Arcs {
  emitterUpdateScript [
   &EmitterState(
    Enable Distance Culling=true
    Life Cycle Mode=::NewEnumerator1
    Loop Duration=&RandomRangeFloat(Minimum=0.1, Maximum=2)
    Max Distance Response=::SleepAndClearParticles
    MaxDistance=3000
    MinDistance=1
    Module.bUseMaxDistance=true
    Module.bUseMinDistance=true
    Recalculate Duration Each Loop=true
    Scalability Mode=::NewEnumerator1
    UseLoopDelay=false
   )
   &SpawnBurst_Instantaneous(
    Spawn Count=&MakeCustomIntFromBool(
     Bool=&SetBoolByFloatComparison(A=&RandomRangeFloat(Minimum=0, Maximum=1), B=0.25)
     True INT=1
     False INT=2
    )
    Use Spawn Probability=false
    UseLoopCountLimit=false
   )
   &SpawnRate(SpawnRate=20, Use Spawn Probability=false) {enabled=false}
  ]
  particleSpawnScript [
   &InitializeParticle(
    Color=<rgba 25 0 18.9743 1>
    Lifetime=0.1
    Mesh Scale Max=<0.037 1 1>
    Mesh Scale Min=<0.037 1.25 1.25>
    Mesh Scale Mode=::NewEnumerator2
    Position Mode=::NewEnumerator1
    Sprite Size Mode=::NewEnumerator3
    Uniform Sprite Size=31.3334
    UsePositionOffset=false
    Write Material Random=true
   )
   /SetVariables(Particles.Pos Choice:NiagaraInt32=&Multiply_Int(A=&RandomRangeInt(Minimum=7, Maximum=20), B=2)) {enabled=false, isSetParameters=true, module=SetVariables_#1}
   /SetVariables(Particles.Pos Choice:NiagaraInt32=&Multiply_Int(
     A=&SelectIntFromArray(
      Array Sampling Mode=::NewEnumerator1
      Direct Array Index=?
      Direct Array Mode=::NewEnumerator1
      Int Selection Array=DataInterface() {propertyValues {IntData [0,1,3,4,5]}}
      Random Seed=0
     )
     B=2
    )) {isSetParameters=true}
   &InitialMeshOrientation(
    Mesh Orientation Mode=::NewEnumerator7
    Rotation=&RandomRangeVector(Minimum=<0 0 0>, Maximum=<1 0 0>)
    Use Rotation Vector=true
   )
   /SetVariables(Particles.Pos A Color:LinearColor=<rgba 1 0 0 1>, Particles.PosB COlor:LinearColor=<rgba 0 1 0 1>) {isSetParameters=true}
  ]
  particleUpdateScript [
   &ParticleState()
   &self:Drift_Horror_Get_Set_of_4_Bones(New Skeletal Mesh:NiagaraDataInterfaceSkeletalMesh=$User.SKeletalMesh, StartIndex:NiagaraInt32=$Particles.Initial.Pos Choice) {module=Drift_Horror_Get_Set_of_4_Bones}
   /SetVariables(
    Particles.Pos A:Vector3f=&Lerp_Vector(
     A=$Output.Drift_Horror_Get_Set_of_4_Bones.BonePos1
     B=$Output.Drift_Horror_Get_Set_of_4_Bones.BonePos2
     Alpha=&RandomRangeFloat(Minimum=0, Maximum=1)
    )
    Particles.Pos B:Vector3f=&Lerp_Vector(
     A=$Output.Drift_Horror_Get_Set_of_4_Bones.BonePos3
     B=$Output.Drift_Horror_Get_Set_of_4_Bones.BonePos4
     Alpha=&RandomRangeFloat(Minimum=0, Maximum=1)
    )
   ) {isSetParameters=true, module=SetVariables_#2}
   /SetVariables(Particles.Position:Vector3f=$Particles.Pos A) {isSetParameters=true, module=SetVariables_#3}
   &DynamicMaterialParameters(
    Index 0 Param 1=1
    Index 0 Param 2=0.5
    Index 0 Param 3=-0.1
    Module.Param0WriteEnabled=true
    Module.Param1WriteEnabled=true
    Module.Param2WriteEnabled=true
    Module.Param3WriteEnabled=true
   )
   /Niagara/Modules/Update/Orientation/V2/OrientMeshToVector(
    Look At Direction:Vector3f=&NormalizeVector(
     Vector To Normalize=&Subtract_Vector(A=$Particles.Pos B, B=$Particles.Pos A)
     Safe Normalize=true
     Safe Fallback Threshold=0
     Safe Fallback Vector=<0 0 0.001>
    )
    Override Quat=false
    Up Axis Direction:Vector3f=<0 0 1>
    UseOptionalSideAxis=false
    UseOptionalUpAxis=false
   )
   &ScaleMeshSize(Scale Factor=&MakeVector(
     X=&VectorLength(Vector=&Subtract_Vector(A=$Particles.Pos B, B=$Particles.Pos A))
     Y=1
     Z=1
    ))
   &FN_Light_Attributes(
    Color=<rgba 7 1.75 5.7346 1>
    Create Color Attribute=true
    Create Position Attribute=true
    Position=&Lerp_Vector(A=$Particles.Pos A, B=$Particles.Pos B, Alpha=0.5)
    Scale Color=&FloatFromCurve(FloatCurve=DataInterface() {propertyValues {Curve {keys [
         {arriveTangent=5.7658, interpMode=RCIM_Cubic, leaveTangent=5.7658, tangentMode=RCTM_User, time=0, value=0}
         <at 0.3 0.9>
         {interpMode=RCIM_Cubic, tangentMode=RCTM_User, time=0.5, value=0.5}
         <at 0.8 0.8>
         {arriveTangent=-4.8787, interpMode=RCIM_Cubic, leaveTangent=-4.8787, tangentMode=RCTM_User, time=1, value=0}
        ]}}})
   )
  ]
  rendererClasses [&NiagaraMeshRendererProperties, &NiagaraLightRendererProperties]
  renderers [
   {rendererClass=&NiagaraMeshRendererProperties, rendererIndex=0}
   {rendererClass=&NiagaraLightRendererProperties, rendererIndex=1}
  ]
 }
 Lightning_ou_of_spikes {
  emitterUpdateScript [
   &EmitterState(
    Enable Distance Culling=true
    Life Cycle Mode=::NewEnumerator1
    Loop Duration=&RandomRangeFloat(Minimum=0.1, Maximum=1)
    Max Distance Response=::SleepAndClearParticles
    MaxDistance=3500
    MinDistance=1
    Module.bUseMaxDistance=true
    Module.bUseMinDistance=true
    Recalculate Duration Each Loop=true
    Scalability Mode=::NewEnumerator1
    UseLoopDelay=false
   )
   &SpawnBurst_Instantaneous(
    Spawn Count=&MakeCustomIntFromBool(
     Bool=&SetBoolByFloatComparison(A=&RandomRangeFloat(Minimum=0, Maximum=0.5), B=0.25)
     True INT=1
     False INT=2
    )
    Use Spawn Probability=false
    UseLoopCountLimit=false
   )
   &SpawnRate(SpawnRate=900, Use Spawn Probability=false) {enabled=false}
  ]
  particleSpawnScript [
   &InitializeParticle(
    Color=<rgba 25 0 18.9743 1>
    Lifetime=0.1
    Mesh Scale=&Multiply_VectorByFloat(Vector=&RandomRangeVector(Minimum=<1 1 1>, Maximum=<3.5 3.5 3.5>), Float=0.25)
    Mesh Scale Mode=::NewEnumerator4
    Position Mode=::NewEnumerator1
    Sprite Size Mode=::NewEnumerator3
    Uniform Sprite Size=2
    UsePositionOffset=false
    Write Material Random=true
   )
   /SetVariables(Particles.Pos Choice:NiagaraInt32=&Multiply_Int(A=&RandomRangeInt(Minimum=0, Maximum=6), B=2)) {isSetParameters=true, module=SetVariables_#1}
   &InitialMeshOrientation(Use Rotation Vector=false)
  ]
  particleUpdateScript [
   &ParticleState()
   &self:Drift_Horror_Get_Set_of_4_Bones(New Skeletal Mesh:NiagaraDataInterfaceSkeletalMesh=$User.SKeletalMesh, StartIndex:NiagaraInt32=$Particles.Initial.Pos Choice) {module=Drift_Horror_Get_Set_of_4_Bones}
   /SetVariables(
    Particles.Pos A:Vector3f=&Lerp_Vector(
     A=$Output.Drift_Horror_Get_Set_of_4_Bones.BonePos1
     B=$Output.Drift_Horror_Get_Set_of_4_Bones.BonePos2
     Alpha=&RandomRangeFloat(Minimum=0, Maximum=1)
    )
    Particles.Pos B:Vector3f=&Lerp_Vector(
     A=$Output.Drift_Horror_Get_Set_of_4_Bones.BonePos3
     B=$Output.Drift_Horror_Get_Set_of_4_Bones.BonePos4
     Alpha=&RandomRangeFloat(Minimum=0, Maximum=1)
    )
   ) {isSetParameters=true, module=SetVariables_#2}
   /SetVariables(Particles.Position:Vector3f=$Particles.Pos A) {isSetParameters=true, module=SetVariables_#3}
   &OrientMeshToVector(
    Look At Direction:Vector3f=&Multiply_VectorByFloat(
     Vector=&Subtract_Vector(A=$Output.Drift_Horror_Get_Set_of_4_Bones.BonePos2, B=$Output.Drift_Horror_Get_Set_of_4_Bones.BonePos1)
     Float=-1
    )
    Override Quat=false
    Up Axis Direction:Vector3f=<0 0 1>
    UseOptionalSideAxis=false
    UseOptionalUpAxis=false
   ) {enabled=false}
   &DynamicMaterialParameters(
    Index 0 Param 1=1
    Index 0 Param 3=-0.1
    Module.Param0WriteEnabled=true
    Module.Param1WriteEnabled=true
    Module.Param2WriteEnabled=true
    Module.Param3WriteEnabled=true
   )
   &ScaleMeshSize(Scale Factor=&MakeVector(
     X=&FloatFromCurve(FloatCurve=DataInterface() {propertyValues {Curve {keys [{tangentMode=RCTM_User, time=0, value=0.5}, {tangentMode=RCTM_User, time=0.1, value=1}, <at 1 0.9>]}}})
     Y=1
     Z=1
    ))
   &FN_Light_Attributes(
    Color=<rgba 2 0.5 1.6385 1>
    Create Color Attribute=true
    Create Exponent Attribute=true
    Create Position Attribute=true
    Position=&Add_Vector(
     A=$Particles.Pos A
     B=&Multiply_VectorByFloat(
      Vector=&Subtract_Vector(A=$Output.Drift_Horror_Get_Set_of_4_Bones.BonePos2, B=$Output.Drift_Horror_Get_Set_of_4_Bones.BonePos1)
      Float=-1
     )
    )
    Scale Color=&FloatFromCurve(FloatCurve=DataInterface() {propertyValues {Curve {keys [
         {arriveTangent=5.7658, interpMode=RCIM_Cubic, leaveTangent=5.7658, tangentMode=RCTM_User, time=0, value=0}
         <at 0.3 0.9>
         {interpMode=RCIM_Cubic, tangentMode=RCTM_User, time=0.5, value=0.5}
         <at 0.8 0.8>
         {arriveTangent=-4.8787, interpMode=RCIM_Cubic, leaveTangent=-4.8787, tangentMode=RCTM_User, time=1, value=0}
        ]}}})
   )
  ]
  rendererClasses [&NiagaraMeshRendererProperties, &NiagaraLightRendererProperties]
  renderers [
   {rendererClass=&NiagaraMeshRendererProperties, rendererIndex=0}
   {rendererClass=&NiagaraLightRendererProperties, rendererIndex=1}
  ]
 }
 Ribbon_Tails001 {
  emitterUpdateScript [
   &EmitterState(Inactive Response=::NewEnumerator1, Life Cycle Mode=::NewEnumerator1, Loop Duration=2, UseLoopDelay=false)
   &SpawnRate(
    SpawnRate=&MakeCustomFloatFromBool(
     Bool=$User.isSkydiving
     True Float=0
     False Float=&Lerp_Float(
      A=5
      B=50
      Alpha=&Divide_Float(A=&VectorLength(Vector=$Engine.Owner.Velocity), B=1000)
     )
    )
    Use Spawn Probability=false
   )
  ]
  particleSpawnScript [
   &FN_SetSpawnGroup(Number of Spawn Groups=2, Set Ribbon ID and Ribbon Link Order=true)
   &SampleSkeletalMesh(
    AdditionalYawPitchRoll=false
    Bone / Socket Index:NiagaraInt32=&MaskIntBySpawnGroup(Masked Int=0, Passthrough Int=4, Spawn Group Mask=1)
    Mesh Sampling Type:ENiagaraSkelSamplingModeFull=::NewEnumerator5
    Skeletal Mesh:NiagaraDataInterfaceSkeletalMesh=$User.SkeletalMesh
    Socket Sampling Mode:ENiagaraSocketSamplingMode=::NewEnumerator5
   )
   &InitializeRibbon(
    Color Maximum:LinearColor=<rgba 0.0215 0.005 0.05 0.2>
    Color Minimum:LinearColor=<rgba 0.0022 0.0005 0.005 0.25>
    Color Mode:ENiagara_ColorInitializationMode=::NewEnumerator2
    Lifetime=&MakeCustomFloatFromBool(
     Bool=$User.isSkydiving
     True Float=2
     False Float=&Lerp_Float(
      A=3
      B=2.5
      Alpha=&Divide_Float(A=&VectorLength(Vector=$Engine.Owner.Velocity), B=1000)
     )
    )
    Position Offset:Vector3f=&Subtract_Vector(
     A=Linked() {linkedVariable {name=StackContext.SampleSkeletalMesh.SampledPosition, type=NiagaraPosition}}
     B=/Niagara/DynamicInputs/Helpers/SimulationPosition()
    )
    Ribbon Width=&RandomRangeFloat(Minimum=25, Maximum=20)
    UsePositionOffset=true
    Write Material Random=true
   )
   /SetVariables(
    Particles.RibbonU0Override=&Add_Float(
     A=&Add_Float(
      A=&Multiply_Float(A=$Emitter.Age, B=0.35)
      B=&MaskFloatBySpawnGroup(Masked Float=0, Passthrough Float=0.5, Spawn Group Mask=1)
     )
     B=&RandomRangeFloat(Minimum=-0.05, Maximum=0.05)
    )
    Particles.SpriteSize=<5 5>
   ) {isSetParameters=true}
   &AddVelocity(Velocity=&Multiply_VectorByFloat(
     Vector=&MaskVectorBySpawnGroup(Masked Vector=<-50 -25 -100>, Passthrough Vector=<50 -25 -100>, Spawn Group Mask=1)
     Float=0.1
    ))
   &CameraOffset(Camera Offset Amount=25)
  ]
  particleUpdateScript [
   &ParticleState()
   /SetVariables(Particles.Lifetime=&Multiply_Float(
     A=$Particles.Lifetime
     B=&FloatFromCurve(
      FloatCurve=DataInterface() {propertyValues {Curve {keys [{tangentMode=RCTM_User, time=0, value=0}, {tangentMode=RCTM_User, time=1, value=1}]}}}
      CurveIndex=&MakeCustomFloatFromBool(
       Bool=$User.isSkydiving
       True Float=&Lerp_Float(
        A=1
        B=0.98
        Alpha=&Divide_Float(A=&VectorLength(Vector=$Engine.Owner.Velocity), B=2000)
       )
       False Float=&Lerp_Float(
        A=1
        B=0.975
        Alpha=&Divide_Float(A=&VectorLength(Vector=$Engine.Owner.Velocity), B=1000)
       )
      )
     )
    )) {isSetParameters=true}
   /SetVariables(Particles.RibbonU0Override=&Add_Float(
     A=$Particles.Initial.RibbonU0Override
     B=&Multiply_Float(A=$Emitter.Age, B=&MakeCustomFloatFromBool(Bool=$User.isSkydiving, True Float=0.25, False Float=0))
    )) {isSetParameters=true}
   &InheritSourceMovement(
    Applied Position Scale=&MakeCustomVectorFromBool(
     Bool=$User.isSkydiving
     True Vector=<1 1 1>
     False Vector=&VectorFromFloat(Value=&FloatFromCurve(FloatCurve=DataInterface() {propertyValues {Curve {keys [{arriveTangent=-2.8707, time=0.35, value=1}, <at 1 0>]}}}))
    )
    ApplyForce=false
    ApplyPosition=true
    ApplyPositionFalloff=false
   )
   &InheritVelocity(
    Inherited Velocity Amount Scale=<0.1 0.1 0.1>
    Inherited Velocity Speed Limit=25
    Speed Limit=true
    UseSpeedThreshold=false
    Velocity Source=&Multiply_VectorByFloat(Vector=$Engine.Owner.Velocity, Float=-1)
   )
   &self:clamp_Local_Y_pos(Min=-1) {module=clamp_Local_Y_pos}
   &GravityForce(Gravity=&MakeVector(
     X=0
     Y=0
     Z=&Lerp_Float(
      A=-25
      B=-100
      Alpha=&Divide_Float(A=&VectorLength(Vector=$Engine.Owner.Velocity), B=1000)
     )
    ))
   &SolveForcesAndVelocity(Clamp Velocity=false, Limit Acceleration=false)
   &ScaleColor(
    Scale Alpha=&Multiply_Float(
     A=&Multiply_Float(
      A=&FloatFromCurve(FloatCurve=DataInterface() {propertyValues {Curve {keys [
           {arriveTangent=39.9667, interpMode=RCIM_Cubic, leaveTangent=39.9667, tangentMode=RCTM_User, time=0, value=0}
           {
            arriveTangent=11.2826
            arriveTangentWeight=0.9439
            interpMode=RCIM_Cubic
            leaveTangent=-3.821
            leaveTangentWeight=1.6479
            tangentMode=RCTM_Break
            tangentWeightMode=RCTWM_WeightedBoth
            time=0.25
            value=5
           }
           {
            arriveTangent=-20.7806
            arriveTangentWeight=2.0492
            interpMode=RCIM_Cubic
            leaveTangent=-20.7806
            leaveTangentWeight=2.0492
            tangentMode=RCTM_User
            tangentWeightMode=RCTWM_WeightedBoth
            time=0.75
            value=0
           }
          ]}}})
      B=&FloatFromCurve(
       FloatCurve=DataInterface() {propertyValues {Curve {keys [{arriveTangent=5.0623, interpMode=RCIM_Cubic, leaveTangent=5.0623, tangentMode=RCTM_User, time=0.25, value=0}, <at 0.9 1>, <at 1 0>]}}}
       CurveIndex=&ReturnNormalizedExecIndex()
      )
     )
     B=1
    )
    ScaleA=true
    ScaleRGB=true
   )
   &ScaleRibbonWidth(Ribbon Width Scale=&FloatFromCurve(FloatCurve=DataInterface() {propertyValues {Curve {keys [<at 0 1>,<at 1 0.5>]}}}))
   /SetVariables(Particles.RibbonLinkOrder:NiagaraFloat=&MakeFloatFromInt(Int=&ReturnParticleIDAttributes())) {isSetParameters=true}
   /SetVariables(Particles.Wiggled Position:Vector3f=&Add_Vector(
     A=Linked() {linkedVariable {name=Particles.Position, type=NiagaraPosition}}
     B=&Multiply_VectorByFloat(Vector=&RandomVector(Vector Scale=0), Float=$Particles.NormalizedAge)
    )) {isSetParameters=true}
  ]
  rendererClasses [&NiagaraRibbonRendererProperties]
  renderers [{rendererClass=&NiagaraRibbonRendererProperties}]
 }
 SmokePuff_Gen5 {
  emitterUpdateScript [
   &EmitterState(
    Enable Distance Culling=true
    MaxDistance=4000
    MinDistance=1
    Module.bUseMaxDistance=true
    Module.bUseMinDistance=true
    Scalability Mode=::NewEnumerator1
   )
   &SpawnBurst_Instantaneous(Spawn Count=1, Use Spawn Probability=false)
   &SpawnRate(SpawnRate=2.5, Use Spawn Probability=false)
  ]
  particleSpawnScript [
   &InitializeParticle(
    Color Maximum=<rgba 0.0215 0.005 0.05 0.5>
    Color Minimum=<rgba 0.0323 0.0075 0.075 0.75>
    Color Mode=::NewEnumerator2
    Lifetime Max=4
    Lifetime Min=3
    Lifetime Mode=::NewEnumerator1
    Sprite Rotation Mode=::NewEnumerator2
    Sprite Size Mode=::NewEnumerator1
    Sprite UV Mode=::NewEnumerator3
    Uniform Sprite Size Max=25
    Uniform Sprite Size Min=30
    UsePositionOffset=false
    Write Material Random=true
   ) {module=InitializeParticle001}
   &CameraOffset(Camera Offset Amount=&RandomRangeFloat(Minimum=0, Maximum=10, Evaluation Type=::NewEnumerator1))
   &AddVelocity(Velocity=&Multiply_VectorByFloat(Vector=$Output.SphereLocation.SphereVector, Float=10))
   &CylinderLocation(
    Clamp UV Range=false
    Cylinder Axis:Vector3f=<0 1 -0.1>
    Cylinder Distribution:ENiagara_CylinderMode=::NewEnumerator2
    Cylinder Height=1
    Cylinder Radius=12
    MaskSpawnGroup=false
    Non Uniform Scale:Vector3f=<1 0.7 1>
    Radius Position:NiagaraFloat=&RandomRangeFloat(Minimum=0.85, Maximum=1)
    Rotation Mode:ENiagara_RotationMode=::NewEnumerator1
    U Position:NiagaraFloat=&RandomRangeFloat(Minimum=-0.1, Maximum=0.6)
    UseOwnerScale=true
    V Position=0
   )
  ]
  particleUpdateScript [
   &ParticleState()
   &ScaleSpriteSize(Scale Factor=&Vector2DFromFloat(Value=&FloatFromCurve(FloatCurve=DataInterface() {propertyValues {Curve {keys [
          {arriveTangent=3.2112, interpMode=RCIM_Cubic, leaveTangent=3.2112, tangentMode=RCTM_User, time=0, value=0.1}
          {
           arriveTangentWeight=0.6453
           interpMode=RCIM_Cubic
           leaveTangentWeight=0.6453
           tangentMode=RCTM_User
           tangentWeightMode=RCTWM_WeightedBoth
           time=1
           value=1
          }
         ]}}})))
   &ScaleColor(
    Scale Alpha=&FloatFromCurve(
     FloatCurve=DataInterface() {propertyValues {Curve {keys [
         <at 0 0 RCIM_Cubic>
         {interpMode=RCIM_Cubic, leaveTangent=-2.3157, tangentMode=RCTM_Break, time=0.15, value=1}
         {interpMode=RCIM_Cubic, leaveTangent=-2.2979, time=1, value=0}
        ]}}}
     Scale Curve=8
    )
    ScaleA=true
    ScaleRGB=false
   )
   &GravityForce(Gravity=<0 0 -6>)
   &DynamicMaterialParameters(
    Index 0 Param 1=&FloatFromCurve(
     FloatCurve=DataInterface() {propertyValues {Curve {keys [
         <at 0 0 RCIM_Cubic>
         {arriveTangent=1.5947, interpMode=RCIM_Cubic, leaveTangent=1.5947, tangentMode=RCTM_User, time=0.2557, value=0.5257}
         <at 1 1 RCIM_Cubic>
        ]}}}
     CurveIndex=$Particles.Age
     Scale Curve=0.25
    )
    Index 0 Param 4=1
    Module.Param0WriteEnabled=true
    Module.Param1WriteEnabled=true
    Module.Param2WriteEnabled=true
    Module.Param3WriteEnabled=true
   ) {module=DynamicMaterialParameters001}
   &SolveForcesAndVelocity(Clamp Velocity=false, Limit Acceleration=false)
   &SubUVAnimation(SubUV Loop Count=1.25)
  ]
  rendererClasses [&NiagaraSpriteRendererProperties]
  renderers [{rendererClass=&NiagaraSpriteRendererProperties}]
 }
}
summary {
 systemName=NS_DriftHorror_BackPack
 emitters [
  {
   emitterName=Lightning_Spike_Arcs
   rendererClasses [&NiagaraMeshRendererProperties, &NiagaraLightRendererProperties]
  }
  {
   emitterName=Lightning_ou_of_spikes
   rendererClasses [&NiagaraMeshRendererProperties, &NiagaraLightRendererProperties]
  }
  {emitterName=Ribbon_Tails001, rendererClasses [&NiagaraRibbonRendererProperties]}
  {emitterName=SmokePuff_Gen5, rendererClasses [&NiagaraSpriteRendererProperties]}
 ]
 userVariables [
  User.isSkydiving {type=NiagaraBool}
  User.RibbonIDOffset {type=NiagaraInt32}
  User.SkeletalMesh {
   type=NiagaraDataInterfaceSkeletalMesh
   defaultValue {
    struct=NiagaraExt_VariableValue_DataInterface
    value {dataInterface=&self:NiagaraDataInterfaceSkeletalMesh_0}
   }
  }
 ]
}
