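# Control Rig for the winged Drift Horror backpack cosmetic that procedurally animates its left and right wing bones.
# One execution sequence from BeginExecution runs eight ChainHarmonics units: lower, middle, upper-main and upper-B chains, left side first.
# Each blends reach, sine-eased wave and pendulum motion, with speeds, amplitudes and offsets varied per chain, some differing between sides.
# Debug drawing is on; the hierarchy's pose and movement-state curves are unused by the graph.
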ControlRigBlueprint /BRCosmetics/Accessories/FORT_Backpacks/Backpack_M_MED_Drift_Horror/Meshes/M_MED_Drift_Horror_Pack_CtrlRig
graphs [
 RigVMFunctionLibrary {links [], nodes []}
 RigVMModel {nodes [
   RigUnit_BeginExecution = &RigUnit_BeginExecution()
   /Script/ControlRig.RigUnit_ChainHarmonics(
    ChainRoot=wings_lw_l_01
    DrawWorldOffset=(Rotation=(X=0,Y=0,Z=0,W=1),Translation=(X=0,Y=0,Z=0),Scale3D=(X=1,Y=1,Z=1))
    Pendulum=(bEnabled=True,PendulumStiffness=2,PendulumGravity=(X=0,Y=0,Z=0),PendulumBlend=0.75,PendulumDrag=0.98,PendulumMaximum=0.1,UnwindAxis=(X=0,Y=1,Z=0),UnwindMinimum=0.2,UnwindMaximum=0.05)
    Reach=(bEnabled=True,ReachTarget=(X=0,Y=0,Z=0),ReachAxis=(X=1,Y=0,Z=0))
    Speed=(X=0,Y=0,Z=1.3)
    Wave=(bEnabled=True,WaveFrequency=(X=0,Y=1,Z=3),WaveAmplitude=(X=0,Y=0.5,Z=1.2),WaveOffset=(X=0,Y=0,Z=2.5),WaveNoise=(X=0,Y=0,Z=0),WaveMinimum=0.5,WaveMaximum=1,WaveEase=SineEaseOut)
    WaveCurve=(EditorCurveData=(Keys=((InterpMode=RCIM_Cubic,TangentMode=RCTM_User,Value=0.2,ArriveTangent=0.8047,LeaveTangent=0.1057),(InterpMode=RCIM_Cubic,Time=1,Value=0.3,LeaveTangent=0.3962))))
    bDrawDebug=True
    ExecutePin=@RigUnit_BeginExecution.ExecutePin
   ) {is=struct RigVMUnitNode}
   > &RigUnit_ChainHarmonics(
    ChainRoot=wings_mid_l_01
    DrawWorldOffset=(Rotation=(X=0,Y=0,Z=0,W=1),Translation=(X=0,Y=0,Z=0),Scale3D=(X=1,Y=1,Z=1))
    Pendulum=(bEnabled=True,PendulumStiffness=2,PendulumGravity=(X=0,Y=0,Z=0),PendulumBlend=0.75,PendulumDrag=0.98,PendulumMaximum=0.1,UnwindAxis=(X=0,Y=1,Z=0),UnwindMinimum=0.2,UnwindMaximum=0.05)
    Reach=(bEnabled=True,ReachTarget=(X=0,Y=0,Z=0),ReachAxis=(X=1,Y=0,Z=0))
    Speed=(X=0,Y=1,Z=1.5)
    Wave=(bEnabled=True,WaveFrequency=(X=0,Y=0,Z=3),WaveAmplitude=(X=0,Y=0.2,Z=0.8),WaveOffset=(X=0,Y=0,Z=0),WaveNoise=(X=0,Y=0,Z=0),WaveMinimum=0.5,WaveMaximum=1,WaveEase=SineEaseOut)
    WaveCurve=(EditorCurveData=(Keys=((InterpMode=RCIM_Cubic,TangentMode=RCTM_User,Value=0.3,ArriveTangent=0.8047,LeaveTangent=0.1057),(InterpMode=RCIM_Cubic,Time=1,Value=0.6,LeaveTangent=0.3962))))
    bDrawDebug=True
   )
   > &RigUnit_ChainHarmonics(
    ChainRoot=wings_up_main_l
    DrawWorldOffset=(Rotation=(X=0,Y=0,Z=0,W=1),Translation=(X=0,Y=0,Z=0),Scale3D=(X=1,Y=1,Z=1))
    Pendulum=(bEnabled=True,PendulumStiffness=2,PendulumGravity=(X=0,Y=0,Z=0),PendulumBlend=0.75,PendulumDrag=0.98,PendulumMaximum=0.1,UnwindAxis=(X=0,Y=1,Z=0),UnwindMinimum=0.2,UnwindMaximum=0.05)
    Reach=(bEnabled=True,ReachTarget=(X=0,Y=0,Z=0),ReachAxis=(X=1,Y=0,Z=0))
    Speed=(X=0,Y=1,Z=1.6)
    Wave=(bEnabled=True,WaveFrequency=(X=0,Y=1,Z=2),WaveAmplitude=(X=0,Y=0.3,Z=0.6),WaveOffset=(X=0,Y=0,Z=3),WaveNoise=(X=0,Y=0,Z=0),WaveMinimum=0.3,WaveMaximum=1,WaveEase=SineEaseOut)
    WaveCurve=(EditorCurveData=(Keys=((InterpMode=RCIM_Cubic,TangentMode=RCTM_User,Value=0.3,ArriveTangent=0.8047,LeaveTangent=0.1057),(InterpMode=RCIM_Cubic,Time=1,Value=0.5,LeaveTangent=0.3962))))
    bDrawDebug=True
   )
   > &RigUnit_ChainHarmonics(
    ChainRoot=wings_up_b_l_01
    DrawWorldOffset=(Rotation=(X=0,Y=0,Z=0,W=1),Translation=(X=0,Y=0,Z=0),Scale3D=(X=1,Y=1,Z=1))
    Pendulum=(bEnabled=True,PendulumStiffness=2,PendulumGravity=(X=0,Y=0,Z=0),PendulumBlend=0.75,PendulumDrag=0.98,PendulumMaximum=0.1,UnwindAxis=(X=0,Y=1,Z=0),UnwindMinimum=0.2,UnwindMaximum=0.05)
    Reach=(bEnabled=True,ReachTarget=(X=0,Y=0,Z=0),ReachAxis=(X=1,Y=0,Z=0))
    Speed=(X=0,Y=1,Z=1.8)
    Wave=(bEnabled=True,WaveFrequency=(X=0,Y=1,Z=3),WaveAmplitude=(X=0,Y=0.3,Z=0.6),WaveOffset=(X=0,Y=0,Z=0),WaveNoise=(X=0,Y=0,Z=0),WaveMinimum=0.6,WaveMaximum=1,WaveEase=SineEaseOut)
    WaveCurve=(EditorCurveData=(Keys=((InterpMode=RCIM_Cubic,TangentMode=RCTM_User,Value=0.3,ArriveTangent=0.8047,LeaveTangent=0.1057),(InterpMode=RCIM_Cubic,Time=1,Value=0.5,LeaveTangent=0.3962))))
    bDrawDebug=True
   )
   > &RigUnit_ChainHarmonics(
    ChainRoot=wings_lw_r_01
    DrawWorldOffset=(Rotation=(X=0,Y=0,Z=0,W=1),Translation=(X=0,Y=0,Z=0),Scale3D=(X=1,Y=1,Z=1))
    Pendulum=(bEnabled=True,PendulumStiffness=2,PendulumGravity=(X=0,Y=0,Z=0),PendulumBlend=0.75,PendulumDrag=0.98,PendulumMaximum=0.1,UnwindAxis=(X=0,Y=1,Z=0),UnwindMinimum=0.2,UnwindMaximum=0.05)
    Reach=(bEnabled=True,ReachTarget=(X=0,Y=0,Z=0),ReachAxis=(X=1,Y=0,Z=0))
    Speed=(X=0,Y=0,Z=1.3)
    Wave=(bEnabled=True,WaveFrequency=(X=0,Y=1,Z=3),WaveAmplitude=(X=0,Y=0,Z=1),WaveOffset=(X=0,Y=0,Z=0.3),WaveNoise=(X=0,Y=0,Z=0),WaveMinimum=0.5,WaveMaximum=1,WaveEase=SineEaseOut)
    WaveCurve=(EditorCurveData=(Keys=((InterpMode=RCIM_Cubic,TangentMode=RCTM_User,Value=0.2,ArriveTangent=0.8047,LeaveTangent=0.1057),(InterpMode=RCIM_Cubic,Time=1,Value=0.3,LeaveTangent=0.3962))))
    bDrawDebug=True
   )
   > &RigUnit_ChainHarmonics(
    ChainRoot=wings_mid_r_01
    DrawWorldOffset=(Rotation=(X=0,Y=0,Z=0,W=1),Translation=(X=0,Y=0,Z=0),Scale3D=(X=1,Y=1,Z=1))
    Pendulum=(bEnabled=True,PendulumStiffness=2,PendulumGravity=(X=0,Y=0,Z=0),PendulumBlend=0.75,PendulumDrag=0.98,PendulumMaximum=0.1,UnwindAxis=(X=0,Y=1,Z=0),UnwindMinimum=0.2,UnwindMaximum=0.05)
    Reach=(bEnabled=True,ReachTarget=(X=0,Y=0,Z=0),ReachAxis=(X=1,Y=0,Z=0))
    Speed=(X=0,Y=1,Z=1.5)
    Wave=(bEnabled=True,WaveFrequency=(X=0,Y=0,Z=2),WaveAmplitude=(X=0,Y=0.2,Z=0.8),WaveOffset=(X=0,Y=0,Z=2),WaveNoise=(X=0,Y=0,Z=0),WaveMinimum=0.5,WaveMaximum=1,WaveEase=SineEaseOut)
    WaveCurve=(EditorCurveData=(Keys=((InterpMode=RCIM_Cubic,TangentMode=RCTM_User,Value=0.3,ArriveTangent=0.8047,LeaveTangent=0.1057),(InterpMode=RCIM_Cubic,Time=1,Value=0.6,LeaveTangent=0.3962))))
    bDrawDebug=True
   )
   > &RigUnit_ChainHarmonics(
    ChainRoot=wings_up_main_r
    DrawWorldOffset=(Rotation=(X=0,Y=0,Z=0,W=1),Translation=(X=0,Y=0,Z=0),Scale3D=(X=1,Y=1,Z=1))
    Pendulum=(bEnabled=True,PendulumStiffness=2,PendulumGravity=(X=0,Y=0,Z=0),PendulumBlend=0.75,PendulumDrag=0.98,PendulumMaximum=0.1,UnwindAxis=(X=0,Y=1,Z=0),UnwindMinimum=0.2,UnwindMaximum=0.05)
    Reach=(bEnabled=True,ReachTarget=(X=0,Y=0,Z=0),ReachAxis=(X=1,Y=0,Z=0))
    Speed=(X=0,Y=1,Z=1.6)
    Wave=(bEnabled=True,WaveFrequency=(X=0,Y=1,Z=2),WaveAmplitude=(X=0,Y=0.3,Z=0.3),WaveOffset=(X=0,Y=0,Z=3),WaveNoise=(X=0,Y=0,Z=0),WaveMinimum=0.3,WaveMaximum=1,WaveEase=SineEaseOut)
    WaveCurve=(EditorCurveData=(Keys=((InterpMode=RCIM_Cubic,TangentMode=RCTM_User,Value=0.3,ArriveTangent=0.8047,LeaveTangent=0.1057),(InterpMode=RCIM_Cubic,Time=1,Value=0.5,LeaveTangent=0.3962))))
    bDrawDebug=True
   )
   > &RigUnit_ChainHarmonics(
    ChainRoot=wings_up_b_r_01
    DrawWorldOffset=(Rotation=(X=0,Y=0,Z=0,W=1),Translation=(X=0,Y=0,Z=0),Scale3D=(X=1,Y=1,Z=1))
    Pendulum=(bEnabled=True,PendulumStiffness=2,PendulumGravity=(X=0,Y=0,Z=0),PendulumBlend=0.75,PendulumDrag=0.98,PendulumMaximum=0.1,UnwindAxis=(X=0,Y=1,Z=0),UnwindMinimum=0.2,UnwindMaximum=0.05)
    Reach=(bEnabled=True,ReachTarget=(X=0,Y=0,Z=0),ReachAxis=(X=1,Y=0,Z=0))
    Speed=(X=0,Y=1,Z=1.8)
    Wave=(bEnabled=True,WaveFrequency=(X=0,Y=1,Z=3),WaveAmplitude=(X=0,Y=0.3,Z=0.6),WaveOffset=(X=0,Y=0,Z=0),WaveNoise=(X=0,Y=0,Z=0),WaveMinimum=0.6,WaveMaximum=1,WaveEase=SineEaseOut)
    WaveCurve=(EditorCurveData=(Keys=((InterpMode=RCIM_Cubic,TangentMode=RCTM_User,Value=0.3,ArriveTangent=0.8047,LeaveTangent=0.1057),(InterpMode=RCIM_Cubic,Time=1,Value=0.5,LeaveTangent=0.3962))))
    bDrawDebug=True
   )
  ]}
]
hierarchy [
 root {loc=<0 0 0>, rot=<0 0 0 1>}
 pack {parent=root, loc=<0 -0.7316 0>, rot=<0 0 0 1>}
 wings_base_r {parent=pack, loc=<-1.4046 0.315 -6.16>, rot=<0 -0.7071 0.7071 0>}
 wings_lw_r_01 {parent=wings_base_r, loc=<7.5622 5.9226 -0.3319>, rot=<0.0862 0.3539 -0.5436 -0.7562>}
 wings_lw_r_02 {parent=wings_lw_r_01, loc=<8.7983 1.0915 -0.7257>, rot=<0.0641 0.0664 0.1084 -0.9898>}
 wings_lw_r_03 {parent=wings_lw_r_02, loc=<8.9396 0 0>, rot=<0 0 -0.2762 -0.9611>}
 wings_mid_r_01 {parent=wings_base_r, loc=<12.9586 -0.8984 -0.22>, rot=<0.0144 0.342 -0.2289 -0.9113>}
 wings_mid_r_02 {parent=wings_mid_r_01, loc=<9.176 1.724 -0.0935>, rot=<-0.0468 0.0484 0.2986 -0.952>}
 wings_mid_r_03 {parent=wings_mid_r_02, loc=<8.6339 2.0476 0>, rot=<0 0 -0.6283 -0.778>}
 wings_up_main_r {parent=wings_base_r, loc=<13.479 -13.2112 0.0395>, rot=<-0.2117 0.3227 0.445 -0.8081>}
 wings_up_a_r_01 {parent=wings_up_main_r, loc=<10.7146 2.377 -0.0301>, rot=<-0.0674 0.017 0.1608 -0.9845>}
 wings_up_a_r_02 {parent=wings_up_a_r_01, loc=<10.6463 0.3721 -0.0625>, rot=<0.0083 0.0315 -0.7244 -0.6886>}
 wings_up_a_r_03 {parent=wings_up_a_r_02, loc=<20.7071 0.9819 0>, rot=<0.0144 -0.0306 -0.4252 -0.9045>}
 wings_up_b_r_01 {parent=wings_up_main_r, loc=<6.3212 6.4174 -0.3129>, rot=<-0.0425 0.1504 -0.5448 -0.8238>}
 wings_up_b_r_02 {parent=wings_up_b_r_01, loc=<13.3596 -0.0014 0>, rot=<-0.0008 -0.0024 -0.1975 -0.9803>}
 wings_up_b_r_03 {parent=wings_up_b_r_02, loc=<15.3449 1.4768 0.3004>, rot=<0 0 -0.4417 -0.8972>}
 wings_base_l {parent=pack, loc=<1.4046 0.315 -6.16>, rot=<-0.7071 0 0 -0.7071>}
 wings_lw_l_01 {parent=wings_base_l, loc=<7.5622 -5.9227 -0.3319>, rot=<-0.0862 0.3539 0.5436 -0.7562>}
 wings_lw_l_02 {parent=wings_lw_l_01, loc=<8.7982 -1.0915 -0.7257>, rot=<-0.0641 0.0664 -0.1084 -0.9898>}
 wings_lw_l_03 {parent=wings_lw_l_02, loc=<8.9396 0 0>, rot=<0 0 0.2762 -0.9611>}
 wings_mid_l_01 {parent=wings_base_l, loc=<12.9586 0.8984 -0.22>, rot=<-0.0754 0.3112 0.282 -0.9044>}
 wings_mid_l_02 {parent=wings_mid_l_01, loc=<9.176 -1.7241 -0.0935>, rot=<0.0266 0.1084 -0.3332 -0.9362>}
 wings_mid_l_03 {parent=wings_mid_l_02, loc=<10.7948 -4.9705 0.0001>, rot=<0.0536 0.0858 0.5271 -0.8438>}
 wings_up_main_l {parent=wings_base_l, loc=<13.479 13.2112 0.0395>, rot=<0.2117 0.3227 -0.445 -0.8081>}
 wings_up_a_l_01 {parent=wings_up_main_l, loc=<10.7146 -2.377 -0.0301>, rot=<0.0674 0.017 -0.1608 -0.9845>}
 wings_up_a_l_02 {parent=wings_up_a_l_01, loc=<10.6462 -0.3721 -0.0625>, rot=<-0.0483 -0.0233 0.6148 -0.7869>}
 wings_up_b_l_01 {parent=wings_up_main_l, loc=<6.3212 -6.4174 -0.3129>, rot=<0.05 0.1619 0.5442 -0.8217>}
 wings_up_b_l_02 {parent=wings_up_b_l_01, loc=<13.3596 0.0015 0>, rot=<0.0005 -0.0025 0.0497 -0.9988>}
 wings_up_b_l_03 {parent=wings_up_b_l_02, loc=<15.3448 -1.4768 0.3003>, rot=<0 0 0.4417 -0.8972>}
 Pose_0 {type=Curve, loc=<0 0 0>, rot=<0 0 0 1>}
 Pose_1 {type=Curve, loc=<0 0 0>, rot=<0 0 0 1>}
 Pose_2 {type=Curve, loc=<0 0 0>, rot=<0 0 0 1>}
 Pose_3 {type=Curve, loc=<0 0 0>, rot=<0 0 0 1>}
 Pose_4 {type=Curve, loc=<0 0 0>, rot=<0 0 0 1>}
 Pose_5 {type=Curve, loc=<0 0 0>, rot=<0 0 0 1>}
 Pose_6 {type=Curve, loc=<0 0 0>, rot=<0 0 0 1>}
 Pose_7 {type=Curve, loc=<0 0 0>, rot=<0 0 0 1>}
 Pose_8 {type=Curve, loc=<0 0 0>, rot=<0 0 0 1>}
 Base {type=Curve, loc=<0 0 0>, rot=<0 0 0 1>}
 Sprint {type=Curve, loc=<0 0 0>, rot=<0 0 0 1>}
 Dive {type=Curve, loc=<0 0 0>, rot=<0 0 0 1>}
 Float {type=Curve, loc=<0 0 0>, rot=<0 0 0 1>}
 Targeting {type=Curve, loc=<0 0 0>, rot=<0 0 0 1>}
 Crouch {type=Curve, loc=<0 0 0>, rot=<0 0 0 1>}
 Jump {type=Curve, loc=<0 0 0>, rot=<0 0 0 1>}
 Run {type=Curve, loc=<0 0 0>, rot=<0 0 0 1>}
 Pose_9 {type=Curve, loc=<0 0 0>, rot=<0 0 0 1>}
 Pose_10 {type=Curve, loc=<0 0 0>, rot=<0 0 0 1>}
]
