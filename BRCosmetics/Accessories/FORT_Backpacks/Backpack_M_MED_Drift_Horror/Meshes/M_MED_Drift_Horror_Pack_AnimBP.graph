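# Animation Blueprint for the Drift Horror backpack, posing its wings to follow the player's movement.
# The wingPosition state machine covers ground, jump, fall, targeting, crouch and nested skydive states, with acceleration-driven additive leans on ground and while floating.
# Run and sprint additives layer over it outside skydiving and crouching, then feed a Control Rig, slightly weakened as speed rises, and a root offset.
# A second idle-to-glide state machine is unconnected.
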AnimBlueprint /BRCosmetics/Accessories/FORT_Backpacks/Backpack_M_MED_Drift_Horror/Meshes/M_MED_Drift_Horror_Pack_AnimBP
target_skeleton=/Game/Accessories/FORT_Backpacks/Backpack_M_MED_Drift_Horror/Meshes/M_MED_Drift_Horror_Pack_Skeleton
graphs [
 AnimGraph {nodes [
   n1 = Wing_Movement_Poses_PoseAsset_Add : Run!(PoseWeight:float=1) -> (Pose:Pose Link Structure) {
    is=title AnimGraphNode_PoseByName
    poseAsset=/BRCosmetics/Accessories/FORT_Backpacks/Backpack_M_MED_Drift_Horror/Animation/Wing_Movement_Poses_PoseAsset_Add
    poseName=Run
   }
   n2 = Wing_Movement_Poses_PoseAsset_Add : Sprint!(PoseWeight:float=1) -> (Pose:Pose Link Structure) {
    is=title AnimGraphNode_PoseByName
    poseAsset=&Wing_Movement_Poses_PoseAsset_Add
    poseName=Sprint
   }
   n3 = Get SprintSpeedRemap() -> (SprintSpeedRemap:float)
   n4 = Two Way Blend(
    A=Apply Additive(Base=Use cached pose 'preSprintCache'(), Additive=@n2, Alpha=@n3) {alphaBoolBlend {blendOption=Linear}, alphaScaleBiasClamp {bInterpResult=true, interpSpeedDecreasing=5}}
    B=Apply Additive(Base=Use cached pose 'preSprintCache'(), Additive=@n1, Alpha=Get RunSpeedRemap() -> (RunSpeedRemap:float)) {alphaBoolBlend {blendOption=Linear}, alphaScaleBiasClamp {bInterpResult=true, interpSpeedDecreasing=5}}
    Alpha=@n3
   ) {blendNode {alphaBoolBlend {blendOption=Linear}, alphaScaleBiasClamp {bInterpResult=true, bias=1, interpSpeedDecreasing=5, scale=-1}}}
   n5 = Use cached pose 'preSprintCache'()
   n6 = Get bIsCrouching()
   n7 = Control Rig(
    Source=Blend Poses by bool(
     BlendPose_0=Use cached pose 'preSprintCache'()
     BlendPose_1=Blend Poses by bool(BlendPose_0=@n5, BlendPose_1=@n4, BlendTime_0=0.3, BlendTime_1=0.3, bActiveValue=@n6) {blendType=Linear}
     BlendTime_0=0.3
     BlendTime_1=0.3
     bActiveValue=Get bIsSkyDiving()
    ) {blendType=Linear}
    Alpha=Get CtrlRig_Alpha() -> (CtrlRig_Alpha:double)
   ) {
    alphaBoolBlend {blendOption=Linear}
    controlRigAssetReference {blueprintRigClass=&/M_MED_Drift_Horror_Pack_CtrlRig.M_MED_Drift_Horror_Pack_CtrlRig_C}
   }
   n8 = "Transform (Modify) Bone - Bone: root"(ComponentPose=Local To Component(LocalPose=@n7), Translation="0,-1.5,0", Rotation="0,0,0") {
    translationMode=BMM_Additive
    translationSpace=BCS_BoneSpace
    alphaBoolBlend {blendOption=Linear}
    boneToModify {boneName=root}
    translation {y=-1.5}
   }
   Output Pose(Result=Component To Local(ComponentPose=@n8)) {layerGroup=None}
   AnimGraphNode_StateMachine_0 = wingPosition() {is=title AnimGraphNode_StateMachine}
   Save cached pose 'preSprintCache'(Pose=@AnimGraphNode_StateMachine_0) {cacheName=preSprintCache}
   AnimGraphNode_StateMachine_1 = State Machine() {is=title AnimGraphNode_StateMachine}
   Use cached pose 'preSprintCache'()
  ]}
 wingPosition {
  entry=Base
  path=AnimGraph.AnimGraphNode_StateMachine_0.wingPosition
  ids=id
  nodes [
   AnimStateNode_0 = Base() {
    is=title AnimStateNode
    stateEntered {guid=#1}
    stateFullyBlended {guid=#2}
    stateLeft {guid=#3}
   }
   AnimStateNode_1 = Jumping() {
    is=title AnimStateNode
    stateEntered {guid=#4}
    stateFullyBlended {guid=#5}
    stateLeft {guid=#6}
   }
   AnimStateNode_2 = Falling() {
    is=title AnimStateNode
    stateEntered {guid=#7}
    stateFullyBlended {guid=#8}
    stateLeft {guid=#9}
   }
   AnimStateNode_3 = Targeting() {
    is=title AnimStateNode
    stateEntered {guid=#10}
    stateFullyBlended {guid=#11}
    stateLeft {guid=#12}
   }
   AnimStateNode_5 = SkydiveMode() {
    is=title AnimStateNode
    stateEntered {guid=#13}
    stateFullyBlended {guid=#14}
    stateLeft {guid=#15}
   }
   AnimStateNode_9 = Crouching() {
    is=title AnimStateNode
    stateEntered {guid=#16}
    stateFullyBlended {guid=#17}
    stateLeft {guid=#18}
   }
   AnimStateTransitionNode_0 = AnimStateTransitionNode() {
    blendMode=Custom
    customBlendCurve=/Game/Accessories/FORT_Backpacks/M_MED_Demon_Pack/Animation/JumpTransitionBlendCurve
    priorityOrder=2
    transitionEnd {guid=#19}
    transitionInterrupt {guid=#20}
    transitionStart {guid=#21}
   }
   AnimStateTransitionNode_1 = AnimStateTransitionNode() {
    blendMode=Custom
    crossfadeDuration=0.5
    customBlendCurve=/Game/Accessories/FORT_Backpacks/M_MED_Demon_Pack/Animation/JumpTransitionBlendOutCurve
    transitionEnd {guid=#22}
    transitionInterrupt {guid=#23}
    transitionStart {guid=#24}
   }
   AnimStateTransitionNode_11 = AnimStateTransitionNode() {
    transitionEnd {guid=#25}
    transitionInterrupt {guid=#26}
    transitionStart {guid=#27}
   }
   AnimStateTransitionNode_12 = AnimStateTransitionNode() {
    transitionEnd {guid=#28}
    transitionInterrupt {guid=#29}
    transitionStart {guid=#30}
   }
   AnimStateTransitionNode_13 = AnimStateTransitionNode() {
    transitionEnd {guid=#31}
    transitionInterrupt {guid=#32}
    transitionStart {guid=#33}
   }
   AnimStateTransitionNode_15 = AnimStateTransitionNode() {
    transitionEnd {guid=#34}
    transitionInterrupt {guid=#35}
    transitionStart {guid=#36}
   }
   AnimStateTransitionNode_2 = AnimStateTransitionNode()
   AnimStateTransitionNode_3 = AnimStateTransitionNode() {
    blendMode=Custom
    customBlendCurve=&JumpTransitionBlendOutCurve
    transitionEnd {guid=#37}
    transitionInterrupt {guid=#38}
    transitionStart {guid=#39}
   }
   AnimStateTransitionNode_4 = AnimStateTransitionNode() {
    blendMode=Custom
    customBlendCurve=&JumpTransitionBlendCurve
    transitionEnd {guid=#40}
    transitionInterrupt {guid=#41}
    transitionStart {guid=#42}
   }
   AnimStateTransitionNode_5 = AnimStateTransitionNode() {
    transitionEnd {guid=#43}
    transitionInterrupt {guid=#44}
    transitionStart {guid=#45}
   }
   AnimStateTransitionNode_6 = AnimStateTransitionNode() {
    crossfadeDuration=0.25
    transitionEnd {guid=#46}
    transitionInterrupt {guid=#47}
    transitionStart {guid=#48}
   }
   AnimStateTransitionNode_7 = AnimStateTransitionNode() {
    crossfadeDuration=0.25
    transitionEnd {guid=#49}
    transitionInterrupt {guid=#50}
    transitionStart {guid=#51}
   }
  ]
  links [
   {
    dst=Crouching
    rule=AnimGraph.AnimGraphNode_StateMachine_0.wingPosition.AnimStateTransitionNode_12.AnimationTransitionGraph_1
    src=Base
    via=AnimStateTransitionNode_12
   }
   {
    dst=Falling
    rule=AnimGraph.AnimGraphNode_StateMachine_0.wingPosition.AnimStateTransitionNode_4.AnimationTransitionGraph_1
    src=Base
    via=AnimStateTransitionNode_4
   }
   {
    dst=Jumping
    rule=AnimGraph.AnimGraphNode_StateMachine_0.wingPosition.AnimStateTransitionNode_0.AnimationTransitionGraph_1
    src=Base
    via=AnimStateTransitionNode_0
   }
   {
    dst=SkydiveMode
    rule=AnimGraph.AnimGraphNode_StateMachine_0.wingPosition.AnimStateTransitionNode_15.AnimationTransitionGraph_1
    src=Base
    via=AnimStateTransitionNode_15
   }
   {
    dst=Targeting
    rule=AnimGraph.AnimGraphNode_StateMachine_0.wingPosition.AnimStateTransitionNode_6.AnimationTransitionGraph_1
    src=Base
    via=AnimStateTransitionNode_6
   }
   {
    dst=Base
    rule=AnimGraph.AnimGraphNode_StateMachine_0.wingPosition.AnimStateTransitionNode_13.AnimationTransitionGraph_1
    src=Crouching
    via=AnimStateTransitionNode_13
   }
   {
    dst=Base
    rule=AnimGraph.AnimGraphNode_StateMachine_0.wingPosition.AnimStateTransitionNode_3.AnimationTransitionGraph_1
    src=Falling
    via=AnimStateTransitionNode_3
   }
   {
    dst=Base
    rule=AnimGraph.AnimGraphNode_StateMachine_0.wingPosition.AnimStateTransitionNode_1.AnimationTransitionGraph_1
    src=Jumping
    via=AnimStateTransitionNode_1
   }
   {
    dst=Falling
    rule=AnimGraph.AnimGraphNode_StateMachine_0.wingPosition.AnimStateTransitionNode_2.AnimationTransitionGraph_1
    src=Jumping
    via=AnimStateTransitionNode_2
   }
   {
    dst=SkydiveMode
    rule=AnimGraph.AnimGraphNode_StateMachine_0.wingPosition.AnimStateTransitionNode_5.AnimationTransitionGraph_1
    src=Jumping
    via=AnimStateTransitionNode_5
   }
   {
    dst=Base
    rule=AnimGraph.AnimGraphNode_StateMachine_0.wingPosition.AnimStateTransitionNode_11.AnimationTransitionGraph_1
    src=SkydiveMode
    via=AnimStateTransitionNode_11
   }
   {
    dst=Base
    rule=AnimGraph.AnimGraphNode_StateMachine_0.wingPosition.AnimStateTransitionNode_7.AnimationTransitionGraph_1
    src=Targeting
    via=AnimStateTransitionNode_7
   }
  ]
 }
 Base {
  path=AnimGraph.AnimGraphNode_StateMachine_0.wingPosition.AnimStateNode_0.Base
  nodes [
   Apply Additive() {alphaBoolBlend {blendOption=Linear}}
   Control Rig(Alpha=Get CtrlRig_Alpha()) {
    alphaBoolBlend {blendOption=Linear}
    controlRigAssetReference {blueprintRigClass=&M_MED_Drift_Horror_Pack_CtrlRig_C}
   }
   Wing_Movement_Poses_PoseAsset : Base!(PoseWeight:float=1) -> (Pose:Pose Link Structure) {
    is=title AnimGraphNode_PoseByName
    poseAsset=/BRCosmetics/Accessories/FORT_Backpacks/Backpack_M_MED_Drift_Horror/Animation/Wing_Movement_Poses_PoseAsset
    poseName=Base
   }
   Wing_Movement_Poses_PoseAsset : Base(PoseWeight=1) {poseAsset=&Wing_Movement_Poses_PoseAsset, poseName=Base}
   n1 = Get PlayerForwardAcceleration() -> (PlayerForwardAcceleration:float)
   n2 = Get PlayerSidewaysAcceleration() -> (PlayerSidewaysAcceleration:float)
   Output Animation Pose(Result=Apply Additive(
     Base=Sequence Player 'Drift_Horror_Male_Idle_Backpack'() {sequence=/BRCosmetics/Accessories/FORT_Backpacks/Backpack_M_MED_Drift_Horror/Animation/Drift_Horror_Male_Idle_Backpack}
     Additive=Blendspace Player 'BS_Wing_Leans_Ground'(X=@n2, Y=@n1) {blendSpace=/BRCosmetics/Accessories/FORT_Backpacks/Backpack_M_MED_Drift_Horror/Animation/BS_Wing_Leans_Ground}
    ) {alphaBoolBlend {blendOption=Linear}})
  ]
 }
 Jumping {
  path=AnimGraph.AnimGraphNode_StateMachine_0.wingPosition.AnimStateNode_1.Jumping
  nodes [
   n1 = Wing_Movement_Poses_PoseAsset : Jump!(PoseWeight:float=1) -> (Pose:Pose Link Structure) {
    is=title AnimGraphNode_PoseByName
    poseAsset=&Wing_Movement_Poses_PoseAsset
    poseName=Jump
   }
   Output Animation Pose(Result=@n1)
  ]
 }
 Falling {
  path=AnimGraph.AnimGraphNode_StateMachine_0.wingPosition.AnimStateNode_2.Falling
  nodes [
   n1 = Wing_Movement_Poses_PoseAsset : Jump(PoseWeight=1) {poseAsset=&Wing_Movement_Poses_PoseAsset, poseName=Jump}
   Output Animation Pose(Result=@n1)
  ]
 }
 Targeting {
  path=AnimGraph.AnimGraphNode_StateMachine_0.wingPosition.AnimStateNode_3.Targeting
  nodes [
   n1 = Wing_Movement_Poses_PoseAsset : Targeting!(PoseWeight:float=1) -> (Pose:Pose Link Structure) {
    is=title AnimGraphNode_PoseByName
    poseAsset=&Wing_Movement_Poses_PoseAsset
    poseName=Targeting
   }
   Output Animation Pose(Result=@n1)
  ]
 }
 SkydiveMode {
  path=AnimGraph.AnimGraphNode_StateMachine_0.wingPosition.AnimStateNode_5.SkydiveMode
  nodes [
   AnimGraphNode_StateMachine_0 = Skydiving State() {is=title AnimGraphNode_StateMachine, bSkipFirstUpdateTransition=false}
   Output Animation Pose(Result=@AnimGraphNode_StateMachine_0)
  ]
 }
 Skydiving State {
  entry=Parachuting
  path=AnimGraph.AnimGraphNode_StateMachine_0.wingPosition.AnimStateNode_5.SkydiveMode.AnimGraphNode_StateMachine_0.Skydiving State
  ids=id
  nodes [
   AnimStateNode_1 = Diving() {is=title AnimStateNode}
   AnimStateNode_2 = Floating() {is=title AnimStateNode}
   AnimStateNode_4 = Parachuting() {is=title AnimStateNode}
   AnimStateTransitionNode_1 = AnimStateTransitionNode() {crossfadeDuration=0.5}
   AnimStateTransitionNode_10 = AnimStateTransitionNode()
   AnimStateTransitionNode_12 = AnimStateTransitionNode()
   AnimStateTransitionNode_13 = AnimStateTransitionNode()
   AnimStateTransitionNode_3 = AnimStateTransitionNode()
   AnimStateTransitionNode_5 = AnimStateTransitionNode() {crossfadeDuration=0.5}
  ]
  links [
   {
    dst=Floating
    rule=AnimGraph.AnimGraphNode_StateMachine_0.wingPosition.AnimStateNode_5.SkydiveMode.AnimGraphNode_StateMachine_0.Skydiving State.AnimStateTransitionNode_1.AnimationTransitionGraph_0
    src=Diving
    via=AnimStateTransitionNode_1
   }
   {
    dst=Parachuting
    rule=AnimGraph.AnimGraphNode_StateMachine_0.wingPosition.AnimStateNode_5.SkydiveMode.AnimGraphNode_StateMachine_0.Skydiving State.AnimStateTransitionNode_10.AnimationTransitionGraph_1
    src=Diving
    via=AnimStateTransitionNode_10
   }
   {
    dst=Diving
    rule=AnimGraph.AnimGraphNode_StateMachine_0.wingPosition.AnimStateNode_5.SkydiveMode.AnimGraphNode_StateMachine_0.Skydiving State.AnimStateTransitionNode_5.AnimationTransitionGraph_0
    src=Floating
    via=AnimStateTransitionNode_5
   }
   {
    dst=Parachuting
    rule=AnimGraph.AnimGraphNode_StateMachine_0.wingPosition.AnimStateNode_5.SkydiveMode.AnimGraphNode_StateMachine_0.Skydiving State.AnimStateTransitionNode_3.AnimationTransitionGraph_1
    src=Floating
    via=AnimStateTransitionNode_3
   }
   {
    dst=Diving
    rule=AnimGraph.AnimGraphNode_StateMachine_0.wingPosition.AnimStateNode_5.SkydiveMode.AnimGraphNode_StateMachine_0.Skydiving State.AnimStateTransitionNode_13.AnimationTransitionGraph_1
    src=Parachuting
    via=AnimStateTransitionNode_13
   }
   {
    dst=Floating
    rule=AnimGraph.AnimGraphNode_StateMachine_0.wingPosition.AnimStateNode_5.SkydiveMode.AnimGraphNode_StateMachine_0.Skydiving State.AnimStateTransitionNode_12.AnimationTransitionGraph_1
    src=Parachuting
    via=AnimStateTransitionNode_12
   }
  ]
 }
 Diving {
  path=AnimGraph.AnimGraphNode_StateMachine_0.wingPosition.AnimStateNode_5.SkydiveMode.AnimGraphNode_StateMachine_0.Skydiving State.AnimStateNode_1.Diving
  nodes [
   Apply Additive(Additive=Blendspace Player 'BS_Wing_Leans_Diving'(X=Get PlayerSidewaysAcceleration(), Y=Get PlayerForwardAcceleration()) {blendSpace=/Game/Accessories/FORT_Backpacks/M_MED_Demon_Pack/Animation/OLD/BS_Wing_Leans_Diving}) {alphaBoolBlend {blendOption=Linear}}
   n1 = Wing_Movement_Poses_PoseAsset : Dive!(PoseWeight:float=1) -> (Pose:Pose Link Structure) {
    is=title AnimGraphNode_PoseByName
    poseAsset=&Wing_Movement_Poses_PoseAsset
    poseName=Dive
   }
   Output Animation Pose(Result=@n1)
  ]
 }
 Floating {
  path=AnimGraph.AnimGraphNode_StateMachine_0.wingPosition.AnimStateNode_5.SkydiveMode.AnimGraphNode_StateMachine_0.Skydiving State.AnimStateNode_2.Floating
  nodes [
   n1 = Get PlayerForwardAcceleration()
   n2 = Get PlayerSidewaysAcceleration()
   n3 = Wing_Movement_Poses_PoseAsset : Float!(PoseWeight:float=1) -> (Pose:Pose Link Structure) {
    is=title AnimGraphNode_PoseByName
    poseAsset=&Wing_Movement_Poses_PoseAsset
    poseName=Float
   }
   Output Animation Pose(Result=Apply Additive(
     Base=@n3
     Additive=Blendspace Player 'BS_Wing_Leans_Air'(X=@n2, Y=@n1) {blendSpace=/BRCosmetics/Accessories/FORT_Backpacks/Backpack_M_MED_Drift_Horror/Animation/BS_Wing_Leans_Air}
    ) {alphaBoolBlend {blendOption=Linear}})
  ]
 }
 Parachuting {
  path=AnimGraph.AnimGraphNode_StateMachine_0.wingPosition.AnimStateNode_5.SkydiveMode.AnimGraphNode_StateMachine_0.Skydiving State.AnimStateNode_4.Parachuting
  nodes [
   Apply Additive(
    Additive=Blendspace Player 'BS_Wing_Leans_Parachuting'(X=Get PlayerSidewaysAcceleration(), Y=Get PlayerForwardAcceleration()) {blendSpace=/Game/Accessories/FORT_Backpacks/Backpack_M_MED_Drift_Horror/Animation/BS_Wing_Leans_Parachuting}
    Alpha=0.8
   ) {alphaBoolBlend {blendOption=Linear}}
   Wing_Movement_Poses_PoseAsset : Float(PoseWeight=0.8) {poseAsset=&Wing_Movement_Poses_PoseAsset, poseName=Float}
   n1 = Sequence Player 'Drift_Horror_Male_to_Glide_Backpack'() {
    bLoopAnimation=false
    sequence=/BRCosmetics/Accessories/FORT_Backpacks/Backpack_M_MED_Drift_Horror/Animation/Drift_Horror_Male_to_Glide_Backpack
   }
   Output Animation Pose(Result=@n1)
  ]
 }
 AnimationTransitionGraph_0 {
  path=AnimGraph.AnimGraphNode_StateMachine_0.wingPosition.AnimStateNode_5.SkydiveMode.AnimGraphNode_StateMachine_0.Skydiving State.AnimStateTransitionNode_1.AnimationTransitionGraph_0
  nodes [Result(bCanEnterTransition=NOT Boolean(A=Get bIsSkydiveDiveMode()))]
 }
 AnimationTransitionGraph_1 {
  path=AnimGraph.AnimGraphNode_StateMachine_0.wingPosition.AnimStateNode_5.SkydiveMode.AnimGraphNode_StateMachine_0.Skydiving State.AnimStateTransitionNode_10.AnimationTransitionGraph_1
  nodes [Result(bCanEnterTransition=Get bIsParachuteOpened())]
 }
 AnimationTransitionGraph_1 {
  path=AnimGraph.AnimGraphNode_StateMachine_0.wingPosition.AnimStateNode_5.SkydiveMode.AnimGraphNode_StateMachine_0.Skydiving State.AnimStateTransitionNode_12.AnimationTransitionGraph_1
  nodes [Result(bCanEnterTransition=NOT Boolean(A=Get bIsParachuteOpened()))]
 }
 AnimationTransitionGraph_1 {
  path=AnimGraph.AnimGraphNode_StateMachine_0.wingPosition.AnimStateNode_5.SkydiveMode.AnimGraphNode_StateMachine_0.Skydiving State.AnimStateTransitionNode_13.AnimationTransitionGraph_1
  nodes [Result(bCanEnterTransition=NOT Boolean(A=Get bIsParachuteOpened()))]
 }
 AnimationTransitionGraph_1 {
  path=AnimGraph.AnimGraphNode_StateMachine_0.wingPosition.AnimStateNode_5.SkydiveMode.AnimGraphNode_StateMachine_0.Skydiving State.AnimStateTransitionNode_3.AnimationTransitionGraph_1
  nodes [Result(bCanEnterTransition=Get bIsParachuteOpened())]
 }
 AnimationTransitionGraph_0 {
  path=AnimGraph.AnimGraphNode_StateMachine_0.wingPosition.AnimStateNode_5.SkydiveMode.AnimGraphNode_StateMachine_0.Skydiving State.AnimStateTransitionNode_5.AnimationTransitionGraph_0
  nodes [Result(bCanEnterTransition=Get bIsSkydiveDiveMode())]
 }
 Crouching {
  path=AnimGraph.AnimGraphNode_StateMachine_0.wingPosition.AnimStateNode_9.Crouching
  nodes [
   n1 = Wing_Movement_Poses_PoseAsset : Crouch!(PoseWeight:float=1) -> (Pose:Pose Link Structure) {
    is=title AnimGraphNode_PoseByName
    poseAsset=&Wing_Movement_Poses_PoseAsset
    poseName=Crouch
   }
   Output Animation Pose(Result=@n1)
  ]
 }
 AnimationTransitionGraph_1 {
  path=AnimGraph.AnimGraphNode_StateMachine_0.wingPosition.AnimStateTransitionNode_0.AnimationTransitionGraph_1
  nodes [Result(bCanEnterTransition=Get bIsJumping())]
 }
 AnimationTransitionGraph_1 {
  path=AnimGraph.AnimGraphNode_StateMachine_0.wingPosition.AnimStateTransitionNode_1.AnimationTransitionGraph_1
  nodes [
   Result(bCanEnterTransition=NOT Boolean(A=Get bIsJumping()))
   Get bShouldPredictLanding() -> (bShouldPredictLanding:bool)
   Get bLandingPredicted() -> (bLandingPredicted:bool)
   Get bDebugLandPrediction() -> (bDebugLandPrediction:bool)
   Get bIsFalling()
  ]
 }
 AnimationTransitionGraph_1 {
  path=AnimGraph.AnimGraphNode_StateMachine_0.wingPosition.AnimStateTransitionNode_11.AnimationTransitionGraph_1
  nodes [Result(bCanEnterTransition=NOT Boolean(A=Get bIsSkyDiving()))]
 }
 AnimationTransitionGraph_1 {
  path=AnimGraph.AnimGraphNode_StateMachine_0.wingPosition.AnimStateTransitionNode_12.AnimationTransitionGraph_1
  nodes [Result(bCanEnterTransition=Get bIsCrouching())]
 }
 AnimationTransitionGraph_1 {
  path=AnimGraph.AnimGraphNode_StateMachine_0.wingPosition.AnimStateTransitionNode_13.AnimationTransitionGraph_1
  nodes [Result(bCanEnterTransition=NOT Boolean(A=Get bIsCrouching()))]
 }
 AnimationTransitionGraph_1 {
  path=AnimGraph.AnimGraphNode_StateMachine_0.wingPosition.AnimStateTransitionNode_15.AnimationTransitionGraph_1
  nodes [Result(bCanEnterTransition=Get bIsSkyDiving())]
 }
 AnimationTransitionGraph_1 {
  path=AnimGraph.AnimGraphNode_StateMachine_0.wingPosition.AnimStateTransitionNode_2.AnimationTransitionGraph_1
  nodes [Result(bCanEnterTransition=Get bIsFalling())]
 }
 AnimationTransitionGraph_1 {
  path=AnimGraph.AnimGraphNode_StateMachine_0.wingPosition.AnimStateTransitionNode_3.AnimationTransitionGraph_1
  nodes [Result(bCanEnterTransition=NOT Boolean(A=Get bIsFalling()))]
 }
 AnimationTransitionGraph_1 {
  path=AnimGraph.AnimGraphNode_StateMachine_0.wingPosition.AnimStateTransitionNode_4.AnimationTransitionGraph_1
  nodes [Result(bCanEnterTransition=Get bIsFalling())]
 }
 AnimationTransitionGraph_1 {
  path=AnimGraph.AnimGraphNode_StateMachine_0.wingPosition.AnimStateTransitionNode_5.AnimationTransitionGraph_1
  nodes [Result(bCanEnterTransition=Get bIsSkyDiving())]
 }
 AnimationTransitionGraph_1 {
  path=AnimGraph.AnimGraphNode_StateMachine_0.wingPosition.AnimStateTransitionNode_6.AnimationTransitionGraph_1
  nodes [Result(bCanEnterTransition=Get bIsTargeting())]
 }
 AnimationTransitionGraph_1 {
  path=AnimGraph.AnimGraphNode_StateMachine_0.wingPosition.AnimStateTransitionNode_7.AnimationTransitionGraph_1
  nodes [Result(bCanEnterTransition=NOT Boolean(A=Get bIsTargeting()))]
 }
 State Machine {
  entry=Idle
  path=AnimGraph.AnimGraphNode_StateMachine_1.State Machine
  ids=id
  nodes [
   AnimStateNode_0 = Idle_1() {is=title AnimStateNode}
   AnimStateNode_1 = to_Glide() {is=title AnimStateNode}
   AnimStateNode_2 = Idle()
   AnimStateTransitionNode_4 = AnimStateTransitionNode() {bAutomaticRuleBasedOnSequencePlayerInState=true}
   AnimStateTransitionNode_5 = AnimStateTransitionNode() {bAutomaticRuleBasedOnSequencePlayerInState=true}
   AnimStateTransitionNode_6 = AnimStateTransitionNode() {bAutomaticRuleBasedOnSequencePlayerInState=true}
  ]
  links [
   {
    dst=Idle_1
    rule=AnimGraph.AnimGraphNode_StateMachine_1.State Machine.AnimStateTransitionNode_4.Transition
    src=Idle
    via=AnimStateTransitionNode_4
   }
   {
    dst=to_Glide
    rule=AnimGraph.AnimGraphNode_StateMachine_1.State Machine.AnimStateTransitionNode_5.Transition
    src=Idle_1
    via=AnimStateTransitionNode_5
   }
   {
    dst=Idle
    rule=AnimGraph.AnimGraphNode_StateMachine_1.State Machine.AnimStateTransitionNode_6.Transition
    src=to_Glide
    via=AnimStateTransitionNode_6
   }
  ]
 }
 Idle_1 {
  path=AnimGraph.AnimGraphNode_StateMachine_1.State Machine.AnimStateNode_0.Idle_1
  nodes [Output Animation Pose(Result=Sequence Player 'Drift_Horror_Male_Idle_Backpack'() {sequence=&Drift_Horror_Male_Idle_Backpack})]
 }
 to_Glide {
  path=AnimGraph.AnimGraphNode_StateMachine_1.State Machine.AnimStateNode_1.to_Glide
  nodes [
   n1 = Sequence Player 'Drift_Horror_Male_to_Glide_Backpack'() {
    bLoopAnimation=false
    sequence=&Drift_Horror_Male_to_Glide_Backpack
   }
   Output Animation Pose(Result=@n1)
  ]
 }
 Idle {
  path=AnimGraph.AnimGraphNode_StateMachine_1.State Machine.AnimStateNode_2.Idle
  nodes [Output Animation Pose(Result=Sequence Player 'Drift_Horror_Male_Idle_Backpack'() {sequence=&Drift_Horror_Male_Idle_Backpack})]
 }
 Transition {path=AnimGraph.AnimGraphNode_StateMachine_1.State Machine.AnimStateTransitionNode_4.Transition, nodes [Result()]}
 Transition {path=AnimGraph.AnimGraphNode_StateMachine_1.State Machine.AnimStateTransitionNode_5.Transition, nodes [Result()]}
 Transition {path=AnimGraph.AnimGraphNode_StateMachine_1.State Machine.AnimStateTransitionNode_6.Transition, nodes [Result()]}
 EventGraph {nodes [
   Event BlueprintUpdateAnimation()
   > Set CtrlRig_Alpha(CtrlRig_Alpha:double=MapRangeClamped(Value=Get PawnSpeed2D(), InRangeA=0, InRangeB=200, OutRangeA=1, OutRangeB=0.8)) -> (Output_Get:double)
   TryGetPawnOwner()
  ]}
]
variables [CtrlRig_Alpha {type=number}]
